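# Cloth-shaded, anime-style master material for the Apprentice pickaxe cosmetic, built on material attributes and enabled for skeletal meshes.
# Shared character PreFX output runs through MF_AnimeShading, then the Apprentice animated-triangles function.
# A static switch optionally adds panning emissive, with selectable channels of the anime function's FX output driving alpha and UV offset.
# The shared PostFX function finishes the chain.
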Material /BRCosmetics/Weapons/FORT_Melee/Pickaxe_Apprentice/Materials/M_Pickaxe_Apprentice_AnimeShading
graph {
 properties {
  bAutomaticallySetUsageInEditor=false
  bUseMaterialAttributes=true
  bUsedWithClothing=true
  bUsedWithMeshDeformer=true
  bUsedWithSkeletalMesh=true
  pixelDepthOffsetMode=PDOM_Legacy
  shadingModel=MSM_Cloth
  usedShadingModels=Cloth
 }
 nodes [
  n1 = MaterialFunctionCall() {materialFunction=&BaseFN_CharacterFunction_PreFX}
  n2 = MaterialFunctionCall!(
   Material=@n1.Result
   LitDiffuseOverride
   SSC_AlphaOverride
   Tint Mask
   DF InkLine Visibility Mask
   DF Inkline Injection
   DF Inkline Color Override
   DF Inkline Color Override Mask
   CustomUVs
  ) {materialFunction=/Game/Packages/Fortress_SharedMaterials/Anime_Master_Material/MF_AnimeShading}
  n3 = StaticComponentMaskParameter(None=@n2.FX) {defaultR=true, group=Skin FX, parameterName=PanningEmissive_Channel}
  n4 = StaticComponentMaskParameter(None=@n2.FX) {defaultB=true, group=Skin FX, parameterName=PanningEmissive_UVOffset_Channel}
  n5 = MaterialFunctionCall!(MA=@n2.Result, Shading Curve=@n2.Result, InkLine Mask=@n2.Result, FX=@n2.Result) {materialFunction=/Game/Characters/Player/Male/Medium/Bodies/M_MED_Apprentice/Materials/MF_Apprentice_AnimatedTriangles}
  MaterialFunctionCall(
   AO=@n1.AO
   Distance Blend=@n1.AO
   Fresnel=@n1.AO
   Gmap=@n1.AO
   Pre Skinned Local Position=@n1.AO
   Material=StaticSwitchParameter(
    True=MaterialFunctionCall!(Alpha=@n3, isVehicle, Gmap, UVOffset=@n4, Local Normal, Local Position, EmissiveColorOverride, Material=@n5.Result) {materialFunction=/Game/Weapons/WeaponSkins/MaterialFunction/PanningEmissive}.Result
    False=@n5.Result
   ) {group=Skin FX, parameterName=Use Panning Emissive}
  ) {materialFunction=&BaseFN_CharacterFunction_PostFX}
 ]
}
